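# PistolCanary_mas1000_aud is the audio Level Sequence for mas1000, driving music state on the PistolCanary_AudioManager actor.
# It holds Event Mix ON true and steps Music Phase from 4 through 10 over the timeline.
# Four trigger events fire at or within a few frames of some phase changes.
# A MetaSoundSeq5.1 audio track plays four partly overlapping sections, starting on or shortly after later phase changes.
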LevelSequence /PistolCanary/Sequence/mas/mas1000/PistolCanary_mas1000_aud
bindings [PistolCanary_AudioManager {
  id=#1
  possessed_class=PistolCanary_AudioManager_C
  tracks [
   {
    type=MovieSceneEventTrack
    sections [{
      type=MovieSceneEventTriggerSection
      channels [None {
        type=MovieSceneScriptingEventChannel
        keys [
         [3671, "<Struct 'MovieSceneEvent' (0x852f46afbe0) {}>"]
         [9939, "<Struct 'MovieSceneEvent' (0x852f46afc80) {}>"]
         [10410, "<Struct 'MovieSceneEvent' (0x852f46afd20) {}>"]
         [14660, "<Struct 'MovieSceneEvent' (0x852f46afdc0) {}>"]
        ]
       }]
     }]
   }
   {
    type=MovieSceneBoolTrack
    display_name=Event Mix ON
    sections [{type=MovieSceneBoolSection, channels [None {type=MovieSceneScriptingBoolChannel, default=true}]}]
   }
   {
    type=MovieSceneIntegerTrack
    display_name=Music Phase
    sections [{
      type=MovieSceneIntegerSection
      channels [None {type=MovieSceneScriptingIntegerChannel, default=4, keys [[3670,5],[6723,6],[7170,7],[9938,8],[10407,9],[14660,10]]}]
     }]
   }
  ]
 }]
display_rate {denominator=1, numerator=60}
tick_resolution {denominator=1, numerator=24000}
tracks [{
  type=MovieSceneAudioTrack
  display_name=MetaSoundSeq5.1
  sections [
   {
    type=MovieSceneAudioSection
    end_frame=7276
    start_frame=6723
    channels [
     Volume {type=MovieSceneScriptingFloatChannel, default=1}
     Pitch {type=MovieSceneScriptingFloatChannel, default=1}
     AttachActor {type=MovieSceneScriptingActorReferenceChannel, default="<Struct 'MovieSceneObjectBindingID' (0x852f8b528e0) {}>"}
    ]
   }
   {
    type=MovieSceneAudioSection
    end_frame=10624
    start_frame=9958
    channels [
     Volume {type=MovieSceneScriptingFloatChannel, default=1}
     Pitch {type=MovieSceneScriptingFloatChannel, default=1}
     AttachActor {type=MovieSceneScriptingActorReferenceChannel, default="<Struct 'MovieSceneObjectBindingID' (0x852f8b52c80) {}>"}
    ]
   }
   {
    type=MovieSceneAudioSection
    end_frame=10229
    start_frame=7170
    channels [
     Volume {type=MovieSceneScriptingFloatChannel, default=1}
     Pitch {type=MovieSceneScriptingFloatChannel, default=1}
     AttachActor {type=MovieSceneScriptingActorReferenceChannel, default="<Struct 'MovieSceneObjectBindingID' (0x852f8978370) {}>"}
    ]
   }
   {
    type=MovieSceneAudioSection
    end_frame=14872
    start_frame=10450
    channels [
     Volume {type=MovieSceneScriptingFloatChannel, default=1}
     Pitch {type=MovieSceneScriptingFloatChannel, default=1}
     AttachActor {type=MovieSceneScriptingActorReferenceChannel, default="<Struct 'MovieSceneObjectBindingID' (0x852f8b52ca0) {}>"}
    ]
   }
  ]
 }]
